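# Physics asset for the M_MED_Textile_Pup male player body cosmetic, apparently for secondary motion on its six dyn_pocket bones.
# Capsules cover the pelvis, spine, upper-thigh twist bones and calves, plus small ones on back, front and calf pockets.
# Each pocket is constrained to its thigh or calf as a near-hinge, swinging mainly on one axis.
# Thighs and calves both constrain directly to the pelvis, while spine bodies have no constraints.
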PhysicsAsset /BRCosmetics/Characters/Player/Male/Medium/Bodies/M_MED_Textile_Pup/Meshes/M_MED_Textile_Pup_Physics
bodies {
 calf_l [{shapeType=Capsule, length=24.66, radius=7.7233, center=<-19.7453 0.3032 1.1265>, rotation=<rot 0.2406 87.3577 -87.2494>}]
 calf_r [{shapeType=Capsule, length=24.7029, radius=7.382, center=<19.6605 -0.4534 -0.6276>, rotation=<rot 0 91.209 -88.3732>}]
 dyn_pocket_bk_l [{
   shapeType=Capsule
   length=6.2906
   radius=1.35
   shapeName=dyn_pocket_bk_l_capsule
   center=<0.3451 -0.1183 0.131>
   rotation=<rot 0 176.6857 -64.4005>
  }]
 dyn_pocket_bk_r [{
   shapeType=Capsule
   length=6.2968
   radius=1.35
   shapeName=dyn_pocket_bk_r_capsule
   center=<0.1294 -0.2774 0.1207>
   rotation=<rot 0 -3.4999 -64.6434>
  }]
 dyn_pocket_calf_l [{
   shapeType=Capsule
   length=6.8221
   radius=1.35
   shapeName=dyn_pocket_calf_l_capsule
   center=<-0.2669 -0.2583 0.5901>
   rotation=<rot 8.4005 -89.9993 -2.9658>
  }]
 dyn_pocket_calf_r [{
   shapeType=Capsule
   length=6.7172
   radius=1.35
   shapeName=dyn_pocket_calf_r_capsule
   center=<0.4916 -0.8835 0.2366>
   rotation=<rot -6.7629 -89.9999 2.4749>
  }]
 dyn_pocket_fr_l [{
   shapeType=Capsule
   length=6.5683
   radius=1.35
   shapeName=dyn_pocket_fr_l_capsule
   center=<0.2352 0.7096 0.6789>
   rotation=<rot -11.3228 -90 0.1794>
  }]
 dyn_pocket_fr_r [{
   shapeType=Capsule
   length=6.5683
   radius=1.35
   shapeName=dyn_pocket_fr_r_capsule
   center=<0.0686 0.3457 0.9552>
   rotation=<rot 11.6196 -79.7924 0.7831>
  }]
 pelvis [{shapeType=Capsule, length=15.7856, radius=9.1113, center=<-12.2644 1.6927 0.3757>, rotation=<rot 0 83.5948 1.0168>}]
 spine_01 [{shapeType=Capsule, length=23.25, radius=7.25, shapeName=spine_01_capsule, rotation=<rot 0 -90 0>}]
 spine_02 [{shapeType=Capsule, length=22, radius=7.5, shapeName=spine_02_capsule, center=<0 -0.4155 0>, rotation=<rot 0 -90 0>}]
 spine_03 [{
   shapeType=Capsule
   length=18.0713
   radius=8.7112
   shapeName=spine_03_capsule
   center=<2.5215 0.7558 -0.2635>
   rotation=<rot 5.4683 -89.9999 -2.1188>
  }]
 spine_04 [{
   shapeType=Capsule
   length=18.4228
   radius=7.8265
   shapeName=spine_04_capsule
   center=<2.3623 0.9568 -0.3622>
   rotation=<rot 1.2494 -90 0.9336>
  }]
 thigh_twist_01_l [{shapeType=Capsule, length=23.4852, radius=8.3265, center=<-5.2935 0.1772 0.2929>, rotation=<rot 21.0441 85.1382 -92.8366>}]
 thigh_twist_01_r [{shapeType=Capsule, length=24.2656, radius=8.2678, center=<4.199 -0.286 -1.3055>, rotation=<rot 21.2643 89.866 -92.9869>}]
}
constraints [
 {bone1Name=calf_l, bone2Name=pelvis, swing1LimitDegrees=45, swing2LimitDegrees=45, twistLimitDegrees=45}
 {bone1Name=thigh_twist_01_l, bone2Name=pelvis, swing1LimitDegrees=45, swing2LimitDegrees=45, twistLimitDegrees=45}
 {bone1Name=calf_r, bone2Name=pelvis, swing1LimitDegrees=45, swing2LimitDegrees=45, twistLimitDegrees=45}
 {bone1Name=thigh_twist_01_r, bone2Name=pelvis, swing1LimitDegrees=45, swing2LimitDegrees=45, twistLimitDegrees=45}
 {bone1Name=dyn_pocket_calf_l, bone2Name=calf_l, swing1LimitDegrees=15, swing2LimitDegrees=0.5, twistLimitDegrees=0.5}
 {bone1Name=dyn_pocket_bk_l, bone2Name=thigh_twist_01_l, swing1LimitDegrees=15, swing2LimitDegrees=0.5, twistLimitDegrees=0.5}
 {bone1Name=dyn_pocket_fr_l, bone2Name=thigh_twist_01_l, swing1LimitDegrees=15, swing2LimitDegrees=0.5, twistLimitDegrees=0.5}
 {bone1Name=dyn_pocket_calf_r, bone2Name=calf_r, swing1LimitDegrees=15, swing2LimitDegrees=0.5, twistLimitDegrees=0.5}
 {bone1Name=dyn_pocket_fr_r, bone2Name=thigh_twist_01_r, swing1LimitDegrees=15, swing2LimitDegrees=0.5, twistLimitDegrees=0.5}
 {bone1Name=dyn_pocket_bk_r, bone2Name=thigh_twist_01_r, swing1LimitDegrees=15, swing2LimitDegrees=0.5, twistLimitDegrees=0.5}
]
