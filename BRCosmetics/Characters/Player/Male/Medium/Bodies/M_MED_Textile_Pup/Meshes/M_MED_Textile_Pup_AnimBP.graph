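# Animation blueprint for the Textile Pup male medium body cosmetic, layering secondary physics motion over the pose copied from its attached parent mesh.
# The live chain additively offsets deform_neck, then runs a pelvis-relative RigidBody simulation with the body's own physics asset, scaled-down component motion and overridden gravity.
# The front-pocket AnimDynamics nodes and a second neck transform are unconnected to the output.
# The event graph derives Look_down from head pitch, which nothing shown reads.
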AnimBlueprint /BRCosmetics/Characters/Player/Male/Medium/Bodies/M_MED_Textile_Pup/Meshes/M_MED_Textile_Pup_AnimBP
target_skeleton=&/M_MED_Textile_Pup_Skeleton
graphs [
 AnimGraph {nodes [
   n1 = Anim Dynamics - Bone: dyn_pocket_fr_l(Alpha=1, ExternalForce="0,0,0") {
    angularDampingOverride=0.999
    angularSpringConstant=30000
    bAngularSpring=true
    bGravityOverrideInSimSpace=true
    bOverrideAngularDamping=true
    gravityScale=0.35
    simulationSpace=RootRelative
    alphaBoolBlend {blendOption=Linear}
    boundBone {boneName=dyn_pocket_fr_l}
    physicsBodyDefinitions [{
      sphereCollisionRadius=0
      boundBone {boneName=dyn_pocket_fr_l}
      boxExtents=<5 20 10>
      constraintSetup {angularTargetAxis=AxisY, angularLimitsMax=<0 0 30>, angularLimitsMin=<0 0 -30>, angularTarget=<0 1 0>}
      localJointOffset=<0 -5 0>
     }]
   }
   Anim Dynamics - Bone: dyn_pocket_fr_r(ComponentPose=@n1, Alpha=1, ExternalForce="0,0,0") {
    angularDampingOverride=0.999
    angularSpringConstant=30000
    bAngularSpring=true
    bGravityOverrideInSimSpace=true
    bOverrideAngularDamping=true
    gravityScale=0.35
    simulationSpace=RootRelative
    alphaBoolBlend {blendOption=Linear}
    boundBone {boneName=dyn_pocket_fr_r}
    physicsBodyDefinitions [{
      sphereCollisionRadius=0
      boundBone {boneName=dyn_pocket_fr_r}
      boxExtents=<5 20 10>
      constraintSetup {angularTargetAxis=AxisY, angularLimitsMax=<0 0 30>, angularLimitsMin=<0 0 -30>, angularTarget=<0 1 0>}
      localJointOffset=<0 -5 0>
     }]
   }
   "Transform (Modify) Bone - Bone: deform_neck"(Alpha=1, Translation="0.412,-2.918,-1.427", Rotation="0.000,8,0", Scale="1,1,1") {
    rotationMode=BMM_Additive
    rotationSpace=BCS_BoneSpace
    translationMode=BMM_Additive
    translationSpace=BCS_BoneSpace
    alphaBoolBlend {blendOption=Linear}
    boneToModify {boneName=deform_neck}
    rotation {yaw=8}
    translation=<0.412 -2.918 -1.4272>
   }
   n2 = "Transform (Modify) Bone - Bone: deform_neck"(
    ComponentPose=Local To Component(LocalPose=Copy Pose From Mesh() {bUseAttachedParent=true})
    Alpha=1
    Translation="0.788,-9.269,-3.431"
    Rotation="0.000,8,0"
    Scale="1,1,1"
   ) {
    rotationMode=BMM_Additive
    rotationSpace=BCS_BoneSpace
    translationMode=BMM_Additive
    translationSpace=BCS_BoneSpace
    alphaBoolBlend {blendOption=Linear}
    boneToModify {boneName=deform_neck}
    rotation {yaw=8}
    translation=<0.788 -9.269 -3.431>
   }
   n3 = RigidBody(ComponentPose=@n2) {
    bClampLinearTranslationLimitToRefPose=true
    bOverrideWorldGravity=true
    overridePhysicsAsset=&/M_MED_Textile_Pup_Physics
    simulationSpace=BaseBoneSpace
    alphaBoolBlend {blendOption=Linear}
    baseBoneRef {boneName=pelvis}
    componentAppliedLinearAccClamp=<3000 3000 3000>
    componentLinearAccScale=<0.3 0.3 0.3>
    componentLinearVelScale=<0.25 0.25 0.25>
    overrideWorldGravity {z=-980}
   }
   Output Pose(Result=Component To Local(ComponentPose=@n3)) {layerGroup=None}
  ]}
 EventGraph {nodes [
   Event BlueprintUpdateAnimation()
   n1 = BreakTransform(InTransform=GetDeltaTransformFromRefPose(self=GetOwnerSkeletalMeshComponent(), BoneName=head, BaseName=None))
   > Set Look_down(Look_down:double=MapRangeClamped(Value=BreakRotator(InRot=@n1.Rotation).Pitch, InRangeA=0, InRangeB=5)) -> (Output_Get:double)
  ]}
]
variables [Look_down {type=number}]
