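# Physics asset for the face accessory of the male medium Textile Pup player cosmetic, giving its two dangling strings collision shapes and joints.
# Thin capsules form a five-link left chain and a four-link right chain, each rooted at spine_05 with narrow swing and tighter twist limits.
# Large capsules on the neck and upper spine approximate the torso, with free joints between spine segments and a wider-limit joint at the neck.
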PhysicsAsset /BRCosmetics/Characters/Player/Male/Medium/Bodies/M_MED_Textile_Pup/Meshes/Parts/M_MED_Textile_Pup_FaceAcc_Physics
bodies {
 dyn_string_l_01 [{
   shapeType=Capsule
   length=4.4225
   radius=0.5
   shapeName=dyn_string_l_01_capsule
   center=<2.3777 0.0524 -0.062>
   rotation=<rot 0 94.464 -89.666>
  }]
 dyn_string_l_02 [{
   shapeType=Capsule
   length=4.6695
   radius=0.5145
   shapeName=dyn_string_l_02_capsule
   center=<2.5175 0.0144 -0.0413>
   rotation=<rot 0 91.684 -89.6382>
  }]
 dyn_string_l_03 [{
   shapeType=Capsule
   length=5.255
   radius=0.505
   shapeName=dyn_string_l_03_capsule
   center=<2.9302 0.0389 -0.0512>
   rotation=<rot 0 92.7082 -89.8672>
  }]
 dyn_string_l_04 [{
   shapeType=Capsule
   length=3.255
   radius=0.505
   shapeName=dyn_string_l_04_capsule
   center=<1.4891 -0.0386 0>
   rotation=<rot 0 -89.9999 90.0019>
  }]
 dyn_string_l_05 [{
   shapeType=Capsule
   length=3.5772
   radius=0.8102
   shapeName=dyn_string_l_05_capsule
   center=<1.8493 0.0661 -0.0039>
   rotation=<rot 0 -83.24 -89.9427>
  }]
 dyn_string_r_01 [{
   shapeType=Capsule
   length=4.2892
   radius=0.5232
   shapeName=dyn_string_r_01_capsule
   center=<2.7409 0.0374 0.0933>
   rotation=<rot 0 -87.5339 -89.2867>
  }]
 dyn_string_r_02 [{
   shapeType=Capsule
   length=5.005
   radius=0.505
   shapeName=dyn_string_r_02_capsule
   center=<2.4193 0.0624 0.0715>
   rotation=<rot 0.4352 -89.981 90.0185>
  }]
 dyn_string_r_03 [{
   shapeType=Capsule
   length=6.255
   radius=0.505
   shapeName=dyn_string_r_03_capsule
   center=<3.4833 0.0734 -0.001>
   rotation=<rot 0.0716 -91.3328 -89.1811>
  }]
 dyn_string_r_04 [{
   shapeType=Capsule
   length=2.5056
   radius=0.7377
   shapeName=dyn_string_r_04_capsule
   center=<2.0591 0.0304 -0.0016>
   rotation=<rot 0 90.4099 -89.6426>
  }]
 neck_01 [{shapeType=Capsule, length=12.8644, radius=12.4149, center=<7.6396 -0.1896 -0.3489>, rotation=<rot -25.0848 -100.6751 -85.2628>}]
 spine_02 [{shapeType=Capsule, length=7.505, radius=14.755, shapeName=spine_02_capsule, center=<0 0.3311 0>, rotation=<rot 0 -90 0>}]
 spine_03 [{shapeType=Capsule, length=11.005, radius=12.755, shapeName=spine_03_capsule, center=<0.1148 2.3242 -0.0041>, rotation=<rot 0 -90 0>}]
 spine_04 [{shapeType=Capsule, length=11.255, radius=13.505, shapeName=spine_04_capsule, center=<0 1.9264 -0.0001>, rotation=<rot 0 -90 0>}]
 spine_05 [{shapeType=Capsule, length=11.6972, radius=14.2126, center=<-2.1616 -0.1625 -0.3494>, rotation=<rot -0.3904 81.8799 0.8373>}]
}
constraints [
 {bone1Name=neck_01, bone2Name=spine_05, swing1LimitDegrees=45, swing2LimitDegrees=45, twistLimitDegrees=45}
 {bone1Name=dyn_string_r_01, bone2Name=spine_05, swing1LimitDegrees=20, swing2LimitDegrees=20, twistLimitDegrees=5}
 {bone1Name=dyn_string_r_02, bone2Name=dyn_string_r_01, swing1LimitDegrees=20, swing2LimitDegrees=20, twistLimitDegrees=5}
 {bone1Name=dyn_string_r_03, bone2Name=dyn_string_r_02, swing1LimitDegrees=20, swing2LimitDegrees=20, twistLimitDegrees=5}
 {bone1Name=dyn_string_r_04, bone2Name=dyn_string_r_03, swing1LimitDegrees=20, swing2LimitDegrees=20, twistLimitDegrees=5}
 {bone1Name=dyn_string_l_01, bone2Name=spine_05, swing1LimitDegrees=20, swing2LimitDegrees=20, twistLimitDegrees=5}
 {bone1Name=dyn_string_l_02, bone2Name=dyn_string_l_01, swing1LimitDegrees=20, swing2LimitDegrees=20, twistLimitDegrees=5}
 {bone1Name=dyn_string_l_03, bone2Name=dyn_string_l_02, swing1LimitDegrees=20, swing2LimitDegrees=20, twistLimitDegrees=5}
 {bone1Name=dyn_string_l_04, bone2Name=dyn_string_l_03, swing1LimitDegrees=20, swing2LimitDegrees=20, twistLimitDegrees=5}
 {bone1Name=dyn_string_l_05, bone2Name=dyn_string_l_04, swing1LimitDegrees=20, swing2LimitDegrees=20, twistLimitDegrees=5}
 {
  bone1Name=spine_05
  bone2Name=spine_04
  swing1LimitDegrees=45
  swing1Motion=Free
  swing2LimitDegrees=45
  swing2Motion=Free
  twistLimitDegrees=45
  twistMotion=Free
 }
 {
  bone1Name=spine_04
  bone2Name=spine_03
  swing1LimitDegrees=45
  swing1Motion=Free
  swing2LimitDegrees=45
  swing2Motion=Free
  twistLimitDegrees=45
  twistMotion=Free
 }
 {
  bone1Name=spine_03
  bone2Name=spine_02
  swing1LimitDegrees=45
  swing1Motion=Free
  swing2LimitDegrees=45
  swing2Motion=Free
  twistLimitDegrees=45
  twistMotion=Free
 }
]
