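# Post-process style AnimBlueprint for the Textile Pup face accessory mesh, animating hood, drawstrings and glasses over the pose copied from the attached parent.
# Bone constraints, additive offsets and AnimDynamics on the strings feed a RigidBody simulation, followed by a final left-hood nudge.
# The update event derives hood corrective alphas from head and neck rotation, plus a melee-swing alpha.
# A FrontEnd alpha applies string offsets when rigid body is disabled.
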AnimBlueprint /BRCosmetics/Characters/Player/Male/Medium/Bodies/M_MED_Textile_Pup/Meshes/Parts/M_MED_Textile_Pup_FaceAcc_AnimBP
target_skeleton=&/M_MED_Textile_Pup_FaceAcc_Skeleton
graphs [
 AnimGraph {nodes [
   n1 = Get Is_Swinging() -> (Is_Swinging:double)
   "Transform (Modify) Bone - Bone: dyn_hood_l"(Alpha=@n1, Translation="-20.731,18.765,-1.027", Rotation="0.000,-40.000,-0.000", Scale="1,1,1") {
    rotationMode=BMM_Additive
    rotationSpace=BCS_BoneSpace
    translationMode=BMM_Additive
    translationSpace=BCS_BoneSpace
    alphaBoolBlend {blendOption=Linear}
    boneToModify {boneName=dyn_hood_l}
    rotation {yaw=-40}
    translation=<-20.731 18.765 -1.027>
   }
   n2 = "Constraint - dyn_string_l_01 (PPPP)"!(
    ComponentPose=Local To Component(LocalPose=Copy Pose From Mesh() {bUseAttachedParent=true})
    Alpha=1
    ConstraintWeights_0=0.2347
    ConstraintWeights_1:float=0.08
    ConstraintWeights_2:float=0.4167
    ConstraintWeights_3:float=0.2685
   ) -> (Pose) {
    alphaBoolBlend {blendOption=Linear}
    boneToModify {boneName=dyn_string_l_01}
    constraintSetup [{targetBone {boneName=clavicle_r}}, {targetBone {boneName=spine_03}}, {targetBone {boneName=spine_04}}, {targetBone {boneName=spine_05}}]
    constraintWeights [1,1,1,1]
   }
   n3 = "Constraint - dyn_string_r_01 (PPPP)"!(
    ComponentPose=@n2
    Alpha=1
    ConstraintWeights_0=0.0796
    ConstraintWeights_1:float=0.4206
    ConstraintWeights_2:float=0.2716
    ConstraintWeights_3:float=0.2282
   ) -> (Pose) {
    alphaBoolBlend {blendOption=Linear}
    boneToModify {boneName=dyn_string_r_01}
    constraintSetup [{targetBone {boneName=spine_03}}, {targetBone {boneName=spine_04}}, {targetBone {boneName=spine_05}}, {targetBone {boneName=clavicle_l}}]
    constraintWeights [1,1,1,1]
   }
   n4 = "Constraint - dyn_hood_lower_r (PP)"!(ComponentPose=@n3, Alpha=1, ConstraintWeights_0=0.57, ConstraintWeights_1:float=0.43) -> (Pose) {
    alphaBoolBlend {blendOption=Linear}
    boneToModify {boneName=dyn_hood_lower_r}
    constraintSetup [{targetBone {boneName=spine_05}}, {targetBone {boneName=upperarm_r}}]
    constraintWeights [1,1]
   }
   n5 = "Constraint - dyn_hood_lower_l (PP)"!(ComponentPose=@n4, Alpha=1, ConstraintWeights_0=0.57, ConstraintWeights_1:float=0.43) -> (Pose) {
    alphaBoolBlend {blendOption=Linear}
    boneToModify {boneName=dyn_hood_lower_l}
    constraintSetup [{targetBone {boneName=spine_05}}, {targetBone {boneName=upperarm_l}}]
    constraintWeights [1,1]
   }
   n6 = Get FrontEnd Alpha() -> (FrontEnd Alpha:double)
   n7 = "Transform (Modify) Bone - Bone: dyn_string_r_01"(ComponentPose=@n5, Alpha=@n6, Translation="0,0,0", Rotation="0,5,0", Scale="1,1,1") {
    rotationMode=BMM_Additive
    rotationSpace=BCS_BoneSpace
    translationMode=BMM_Additive
    translationSpace=BCS_BoneSpace
    alphaBoolBlend {blendOption=Linear}
    alphaScaleBiasClamp {inRange {max=0, min=-100}, outRange {max=0, min=1}}
    boneToModify {boneName=dyn_string_r_01}
    rotation {yaw=5}
   }
   n8 = "Transform (Modify) Bone - Bone: dyn_string_l_01"(ComponentPose=@n7, Alpha=@n6, Translation="0,0,0", Rotation="0,5,0", Scale="1,1,1") {
    rotationMode=BMM_Additive
    rotationSpace=BCS_BoneSpace
    translationMode=BMM_Additive
    translationSpace=BCS_BoneSpace
    alphaBoolBlend {blendOption=Linear}
    alphaScaleBiasClamp {inRange {max=0, min=-100}, outRange {max=0, min=1}}
    boneToModify {boneName=dyn_string_l_01}
    rotation {yaw=5}
   }
   n9 = Anim Dynamics - Bone: dyn_string_r_01(ComponentPose=@n8, Alpha=0.5, ExternalForce="0,0,0") {
    angularDampingOverride=1
    bGravityOverrideInSimSpace=true
    bOverrideAngularDamping=true
    gravityScale=0.35
    simulationSpace=RootRelative
    alphaBoolBlend {blendOption=Linear}
    boundBone {boneName=dyn_string_r_01}
    physicsBodyDefinitions [{
      sphereCollisionRadius=0
      boundBone {boneName=dyn_string_r_01}
      boxExtents=<5 10 10>
      constraintSetup {angularLimitsMax=<0 20 0>, angularLimitsMin=<0 -20 -40>, angularTarget=<0 0 0>}
      localJointOffset=<5 0 -0.01>
     }]
   }
   n10 = Anim Dynamics - Bone: dyn_string_l_01(ComponentPose=@n9, Alpha=0.5, ExternalForce="0,0,0") {
    angularDampingOverride=1
    bGravityOverrideInSimSpace=true
    bOverrideAngularDamping=true
    gravityScale=0.35
    simulationSpace=RootRelative
    alphaBoolBlend {blendOption=Linear}
    boundBone {boneName=dyn_string_l_01}
    physicsBodyDefinitions [{
      sphereCollisionRadius=0
      boundBone {boneName=dyn_string_l_01}
      boxExtents=<5 10 10>
      constraintSetup {angularLimitsMax=<0 0 0>, angularLimitsMin=<0 0 -40>, angularTarget=<0 0 0>}
      localJointOffset=<5 0 -0.01>
     }]
   }
   n11 = "Transform (Modify) Bone - Bone: deform_neck"(ComponentPose=@n10, Alpha=1, Translation="-0.732,-3.583,-1.946", Rotation="0.000,8,0", Scale="1,1,1") {
    rotationMode=BMM_Additive
    rotationSpace=BCS_BoneSpace
    translationMode=BMM_Additive
    translationSpace=BCS_BoneSpace
    alphaBoolBlend {blendOption=Linear}
    boneToModify {boneName=deform_neck}
    rotation {yaw=8}
    translation=<-0.732 -3.583 -1.946>
   }
   n12 = "Transform (Modify) Bone - Bone: dyn_hood_r"(
    ComponentPose=@n11
    Alpha=Get head_R_corrective_alpha() -> (head_R_corrective_alpha:double)
    Translation="-2.087,-5.338,-5.887"
    Rotation="-1.292,-4.831,5.053"
    Scale="1,1,1"
   ) {
    rotationMode=BMM_Additive
    rotationSpace=BCS_BoneSpace
    translationMode=BMM_Additive
    translationSpace=BCS_BoneSpace
    alphaBoolBlend {blendOption=Linear}
    boneToModify {boneName=dyn_hood_r}
    rotation=<rot -1.292 -4.831 5.053>
    translation=<-2.0866 -5.338 -5.887>
   }
   n13 = Get head_L_corrective_alpha() -> (head_L_corrective_alpha:double)
   n14 = "Transform (Modify) Bone - Bone: dyn_hood_l"(ComponentPose=@n12, Alpha=@n13, Translation="4.802,-11.562,-4.881", Rotation="0.000,-70.000,-15.000", Scale="1,1,1") {
    rotationMode=BMM_Additive
    rotationSpace=BCS_BoneSpace
    translationMode=BMM_Additive
    translationSpace=BCS_BoneSpace
    alphaBoolBlend {blendOption=Linear}
    boneToModify {boneName=dyn_hood_l}
    rotation {roll=-15.0001, yaw=-70}
    translation=<4.802 -11.562 -4.881>
   }
   n15 = "Transform (Modify) Bone - Bone: dyn_hood_lower_l"(ComponentPose=@n14, Alpha=@n13, Translation="2.735,-3.884,2.045", Rotation="0.000,0.000,5.000", Scale="1,1,1") {
    rotationMode=BMM_Additive
    rotationSpace=BCS_BoneSpace
    translationMode=BMM_Additive
    translationSpace=BCS_BoneSpace
    alphaBoolBlend {blendOption=Linear}
    boneToModify {boneName=dyn_hood_lower_l}
    rotation {roll=5}
    translation=<2.735 -3.884 2.045>
   }
   n16 = "Transform (Modify) Bone - Bone: dyn_hood_r"(ComponentPose=@n15, Alpha=@n1, Translation="-1.543,-1.640,5.521", Rotation="0,0,0", Scale="1,1,1") {
    rotationMode=BMM_Additive
    rotationSpace=BCS_BoneSpace
    translationMode=BMM_Additive
    translationSpace=BCS_BoneSpace
    alphaBoolBlend {blendOption=Linear}
    boneToModify {boneName=dyn_hood_r}
    translation=<-1.543 -1.64 5.521>
   }
   n17 = "Transform (Modify) Bone - Bone: dyn_glasses"(ComponentPose=@n16, Alpha=1, Translation="-0.000,1.2,0.000", Rotation="0,0,0.15", Scale="1,1,1") {
    rotationMode=BMM_Additive
    rotationSpace=BCS_BoneSpace
    translationMode=BMM_Additive
    translationSpace=BCS_BoneSpace
    alphaBoolBlend {blendOption=Linear}
    alphaScaleBiasClamp {inRange {max=0, min=-100}, outRange {max=0, min=1}}
    boneToModify {boneName=dyn_glasses}
    rotation {roll=0.15}
    translation {y=1.2}
   }
   n18 = RigidBody(ComponentPose=@n17) {
    bClampLinearTranslationLimitToRefPose=true
    overridePhysicsAsset=&/M_MED_Textile_Pup_FaceAcc_Physics
    simulationSpace=BaseBoneSpace
    alphaBoolBlend {blendOption=Linear}
    baseBoneRef {boneName=root}
   }
   Output Pose(Result=Component To Local(ComponentPose="Transform (Modify) Bone - Bone: dyn_hood_l"(ComponentPose=@n18, Alpha=1, Translation="2,0,0", Rotation="0,0,0", Scale="1,1,1") {translationMode=BMM_Additive, translationSpace=BCS_BoneSpace, boneToModify {boneName=dyn_hood_l}, translation {x=2}})) {layerGroup=None}
  ]}
 EventGraph {nodes [
   Event BlueprintUpdateAnimation()
   n1 = GetOwnerSkeletalMeshComponent()
   n2 = BreakTransform(InTransform=GetDeltaTransformFromRefPose(self=@n1, BoneName=head, BaseName=None))
   > Set head_R_corrective_alpha(head_R_corrective_alpha:double=MapRangeClamped(Value=BreakRotator(InRot=@n2.Rotation), InRangeA=0, InRangeB=-25)) -> (Output_Get:double)
   n3 = BreakTransform(InTransform=GetDeltaTransformFromRefPose(self=@n1, BoneName=neck_01, BaseName=None))
   > Set head_L_corrective_alpha(head_L_corrective_alpha:double=MapRangeClamped(Value=BreakRotator(InRot=@n3.Rotation), InRangeA=0.1, InRangeB=18, OutRangeA=0.001)) -> (Output_Get:double)
   > n4 = Sequence()
   Set Is_Swinging(execute=@n4.then_1, Is_Swinging:double="To Float (Boolean)"(InBool=Get bIsPlayingMeleeAnim())) -> (Output_Get:double)
   > n5 = Branch(Condition=Get bRigidBodyEnabled())
   > Set FrontEnd Alpha(FrontEnd Alpha:double=0) -> (Output_Get:double)
   Set FrontEnd Alpha(execute=@n5.else, FrontEnd Alpha=1)
   GetAlphaFromReferenceTransformDelta(BoneName=head, BaseName=spine_05, InputRange=(X=0,Y=-45), TwistAxis="0,0,1")
   MapRangeClamped(Value=0, InRangeA=15, InRangeB=90)
   MapRangeClamped(Value=0, InRangeA=40, InRangeB=-120)
   Branch()
   "integer >= integer"(A=Get CurrentLOD(), B=1)
   n6 = BreakTransform(InTransform=GetDeltaTransformFromRefPose(self=GetPartSkeletalMeshComponent(), BoneName=head, BaseName=neck_01))
   Set head_R_corrective_alpha(head_R_corrective_alpha=MapRangeClamped(Value=BreakRotator(InRot=@n6.Rotation), InRangeA=0, InRangeB=-30))
  ]}
]
variables [
 FrontEnd Alpha {type=number}
 head_L_corrective_alpha {type=number}
 head_R_corrective_alpha {type=number}
 Is_Swinging {type=number}
]
